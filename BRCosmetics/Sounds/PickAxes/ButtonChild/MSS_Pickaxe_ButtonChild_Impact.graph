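# One-shot stereo MetaSound for the ButtonChild pickaxe impact sound, layering an impact hit with a boing or cackle.
# On play it rolls random volume and pitch, then picks an impact wave, whose player triggers a second random pick from the accent array.
# Both no-repeat layers share the pitch shift and are summed before volume scaling.
# Only the accent layer signals OnFinished; start delay and template multiplier inputs are unconnected.
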MetaSoundSource /BRCosmetics/Sounds/PickAxes/ButtonChild/MSS_Pickaxe_ButtonChild_Impact
inputs [
 Impact.Randoms {type=WaveAsset:Array}
 Impact.Sounds {type=WaveAsset:Array}
 PitchModulationRange.Max {type=Float}
 PitchModulationRange.Min {type=Float}
 StartDelay.Max {type=Time}
 StartDelay.Min {type=Time}
 TemplatePitchMultiplier {type=Float}
 TemplateVolumeMultiplier {type=Float}
 UE.Source.OnPlay {type=Trigger}
 VolumeModulationRange.Max {type=Float}
 VolumeModulationRange.Min {type=Float}
]
interfaces [UE.OutputFormat.Stereo:1.0, UE.Source.OneShot:1.0, UE.Source:1.1]
outputs [
 "UE.OutputFormat.Stereo.Audio:0" {type=Audio}
 "UE.OutputFormat.Stereo.Audio:1" {type=Audio}
 UE.Source.OneShot.OnFinished {type=Trigger}
]
pages [Default {
  input_defaults {
   PitchModulationRange.Max=1
   PitchModulationRange.Min=0.908
   StartDelay.Max=0
   StartDelay.Min=0
   TemplatePitchMultiplier=1
   TemplateVolumeMultiplier=1
   UE.Source.OnPlay=null
   VolumeModulationRange.Max=1
   VolumeModulationRange.Min=0.956
   Impact.Randoms [
    &/Pickaxe_ButtonChild_Impact_Boing_01
    &/Pickaxe_ButtonChild_Impact_Boing_02
    &/Pickaxe_ButtonChild_Impact_Boing_03
    &/Pickaxe_ButtonChild_Impact_Boing_05
    &/Pickaxe_ButtonChild_Impact_Cackle_01
    &/Pickaxe_ButtonChild_Impact_Cackle_05
    &/Pickaxe_ButtonChild_Impact_Cackle_03
    &/Pickaxe_ButtonChild_Impact_Cackle_04
    &/Pickaxe_ButtonChild_Impact_Boing_07
    &Pickaxe_ButtonChild_Impact_Boing_07
    &Pickaxe_ButtonChild_Impact_Boing_07
   ]
   Impact.Sounds [
    &/Pickaxe_ButtonChild_Impact_01
    &/Pickaxe_ButtonChild_Impact_02
    &/Pickaxe_ButtonChild_Impact_03
    &/Pickaxe_ButtonChild_Impact_04
    &/Pickaxe_ButtonChild_Impact_05
    &/Pickaxe_ButtonChild_Impact_06
   ]
  }
  nodes [
   n1 = UE.RandomFloat(Next=^UE.Source.OnPlay, Max=^VolumeModulationRange.Max, Min=^VolumeModulationRange.Min)
   n2 = UE.RandomFloat(Max=^PitchModulationRange.Max, Min=^PitchModulationRange.Min, Next=@n1.On Next)
   ArrayRandomGet_0 = Array.Random Get:WaveAsset:Array(In Array=^Impact.Sounds, Next=@n2.On Next) {also {defaults {Enable Shared State="<MetaSoundBuilderResult.SUCCEEDED: 0>", No Repeats=4, Weights []}}}
   FrequencyMultipliertoSemitone_0 = UE.Frequency Multiplier to Semitone:Float(Frequency Multiplier=UE.Multiply:Float(PrimaryOperand=@n2.Value).Out)
   WavePlayer_1 = UE.Wave Player:Stereo(
    Play=@ArrayRandomGet_0.On Next
    Wave Asset=@ArrayRandomGet_0.Value
    Pitch Shift=UE.Reroute(Value=@FrequencyMultipliertoSemitone_0.Semitones).Value
   ) {also {version [1,1]}}
   ArrayRandomGet_1 = Array.Random Get:WaveAsset:Array(
    In Array=^Impact.Randoms
    Weights=VariableAccessor.Float:Array(Variable=InitVariable.Float:Array().Variable).Value
    Next=@WavePlayer_1.On Play
   ) {also {defaults {Enable Shared State="<MetaSoundBuilderResult.SUCCEEDED: 0>", No Repeats=2}}}
   WavePlayer_0 = UE.Wave Player:Stereo(Play=@ArrayRandomGet_1.On Next, Wave Asset=@ArrayRandomGet_1.Value, Pitch Shift=@FrequencyMultipliertoSemitone_0.Semitones) {also {version [1,1]}}
   n3 = UE.Reroute(Value=UE.Multiply:Float(PrimaryOperand=@n1.Value) {also {defaults {AdditionalOperands=0.9}}}.Out)
   Multiply_0 = UE.Multiply:Audio by Float(
    PrimaryOperand=UE.Add:Audio(PrimaryOperand=@WavePlayer_0.Out Left, AdditionalOperands=@WavePlayer_1.Out Left).Out
    AdditionalOperands=@n3.Value
   )
   Multiply_1 = UE.Multiply:Audio by Float(
    PrimaryOperand=UE.Add:Audio(PrimaryOperand=@WavePlayer_0.Out Right, AdditionalOperands=@WavePlayer_1.Out Right).Out
    AdditionalOperands=@n3.Value
   )
  ]
  links [
   {dst="@UE.OutputFormat.Stereo.Audio:0", src=Multiply_0.Out}
   {dst="@UE.OutputFormat.Stereo.Audio:1", src=Multiply_1.Out}
   {dst="@UE.Source.OneShot.OnFinished", src=WavePlayer_0.On Finished}
  ]
 }]
properties {
 assetImportData=&self:AssetImportData
 attenuationSettings=/Game/Sounds/Attenuation/Fort_Weapon_Pickaxe_Player_Impact_Att
 audioPropertiesSheet=/BRCosmetics/Sounds/Settings/PropertySheets/Pickaxe/APS_Cosmetic_Pickaxe_Impact
 maxDistance=3512
 numChannels=2
 outputFormat=Stereo
 sampleRate=48000
 soundClassObject=&Fort_Impacts
 volume=0.75
}
